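; Camera actor for the ToyFox pilot cinematic that publishes its current view to the game state.
; Each tick, after the parent tick, an update flag gates a lookup of the ToyFox game state component on the Athena game state.
; When both are valid, it reads the camera component's view and passes it to that component as the pilot cinematic camera view.
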(event EventTick
  (params (DeltaSeconds :type Float%20%28single-precision%29))
  (bind _returnvalue (Fortnite|Athena|GameState|GetGameStateAthena))
  (bind _returnvalue_2 (Actor|GetComponentbyClass _returnvalue "/ToyFoxGameplay/Gameplay/B_ToyFoxGameStateComponent.B_ToyFoxGameStateComponent_C"))
  (|CallParentFunction DeltaSeconds)
  (if (Variables|Default|GetbUpdateToGameStateComponent)
    (Utilities|IsValid _returnvalue
      (:Is%20Valid
        (Utilities|IsValid _returnvalue_2
          (:Is%20Valid
            (bind _desiredview (Camera|GetCameraView (Variables|CameraActor|GetCameraComponent) DeltaSeconds))
            (Cinematics|SetPilotCinematicCameraActorView _returnvalue_2 _desiredview)))))))

(fn ConstructionScript)
