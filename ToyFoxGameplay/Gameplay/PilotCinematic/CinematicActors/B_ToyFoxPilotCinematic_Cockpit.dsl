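; Cockpit actor for the ToyFox pilot cinematic that dresses five mannequins to match the current pilots.
; Its construction script snaps and attaches it to AttachSocket on the first giant player pawn's skeletal mesh.
; On BeginPlay, where cosmetics can run, it maps the game state component's pilot actors to mannequin child actors in a fixed, non-numeric order.
; Player pawns supply character and backpack with variants; secondary pilot stand-ins supply only a character definition.
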(event EventBeginPlay
  ;[0.01171 0.15 0.03614 0.5] BeginPlay: Update the cosmetics of the mannequins.
  (|CallParentFunction)
  (Networking|CanExecuteCosmeticEvents
    (:True
      (CallFunction|SetupMannequins))))

(fn SetupMannequins
  (bind _returnvalue (Fortnite|Athena|GameState|GetGameStateAthena))
  (Utilities|IsValid _returnvalue
    (:Is%20Valid
      (bind _returnvalue_2 (Actor|GetComponentbyClass _returnvalue "/ToyFoxGameplay/Gameplay/B_ToyFoxGameStateComponent.B_ToyFoxGameStateComponent_C"))
      (Utilities|IsValid _returnvalue_2
        (:Is%20Valid
          (bind _pilotactors (Pilots|GetPilotActors _returnvalue_2))
          (for (_array_element _array_index) (Utilities|Array|MakeArray (Variables|Default|GetMannequinChildActor02) (Variables|Default|GetMannequinChildActor03) (Variables|Default|GetMannequinChildActor04) (Variables|Default|GetMannequinChildActor01) (Variables|Default|GetMannequinChildActor05))
            (bind _childactor (Variables|/Script/Engine.ChildActorComponent|GetChildActor _array_element)
              (:then
                (bind _asb_toy_fox_pilot_cinematic_mannequin (Utilities|Casting|CastToB_ToyFoxPilotCinematicMannequin _childactor))
                (LocalVariables|SetMannequin _asb_toy_fox_pilot_cinematic_mannequin)
                (if (Utilities|Array|IsValidIndex _pilotactors _array_index)
                  (bind _asfort_player_pawn (Utilities|Casting|CastToFortPlayerPawn (Utilities|Array|Get(acopy) _pilotactors _array_index))
                    (:then
                      (Customization|SetAthenaBackpackandCharacterDefandVariants (LocalVariables|GetMannequin) (CosmeticLoadout|GetItemDefinitionTypefromLoadoutwithActor _asfort_player_pawn "(TagName=\"LoadoutSlot.Backpack\")" "/Script/FortniteGame.AthenaBackpackItemDefinition") (CosmeticLoadout|GetActiveVariantsForItemFromActor (CosmeticLoadout|GetItemDefinitionTypefromLoadoutwithActor _asfort_player_pawn "(TagName=\"LoadoutSlot.Backpack\")" "/Script/FortniteGame.AthenaBackpackItemDefinition") _asfort_player_pawn) (CosmeticLoadout|GetItemDefinitionTypefromLoadoutwithActor _asfort_player_pawn "(TagName=\"LoadoutSlot.Character\")" "/Script/FortniteGame.AthenaCharacterItemDefinition") (CosmeticLoadout|GetActiveVariantsForItemFromActor (CosmeticLoadout|GetItemDefinitionTypefromLoadoutwithActor _asfort_player_pawn "(TagName=\"LoadoutSlot.Character\")" "/Script/FortniteGame.AthenaCharacterItemDefinition") _asfort_player_pawn)))
                    (:CastFailed
                      (bind _asbga_secondary_pilot_stand_in (Utilities|Casting|CastToBGA_SecondaryPilotStandIn (Utilities|Array|Get(acopy) _pilotactors _array_index)))
                      (Utilities|IsValid (Class|BGASecondaryPilotStandIn|GetCharacterItemDefinition _asbga_secondary_pilot_stand_in)
                        (:Is%20Valid
                          (Customization|ResetAthenaCharacterDefandVariants (LocalVariables|GetMannequin) (Class|BGASecondaryPilotStandIn|GetCharacterItemDefinition _asbga_secondary_pilot_stand_in) (Utilities|Array|MakeArray))))))))
              (:else))))))))

(fn ConstructionScript
  (bind _attachsocket (Variables|Default|GetAttachSocket))
  (bind _outactors (Actor|GetAllActorsOfClass "/ToyFoxGameplay/Gameplay/B_ToyFoxGiantPlayerPawn.B_ToyFoxGiantPlayerPawn_C"))
  (if (Utilities|Array|IsValidIndex _outactors)
    (Utilities|IsValid (Utilities|Array|Get(acopy) _outactors)
      (:Is%20Valid
        (Utilities|IsValid (Components|GetToyFoxSkeletalMeshComponent (Utilities|Array|Get(acopy) _outactors))
          (:Is%20Valid
            (bind (_returnvalue_location _returnvalue_rotation) (Math|Transform|BreakTransform (Transformation|GetSocketTransform (Components|GetToyFoxSkeletalMeshComponent (Utilities|Array|Get(acopy) _outactors)) _attachsocket "RTS_World")))
            (Transformation|SetActorLocationAndRotation _returnvalue_location _returnvalue_rotation)
            (Transformation|AttachActorToComponent (Components|GetToyFoxSkeletalMeshComponent (Utilities|Array|Get(acopy) _outactors)) _attachsocket "SnapToTarget" "KeepWorld" "KeepWorld")))))))
