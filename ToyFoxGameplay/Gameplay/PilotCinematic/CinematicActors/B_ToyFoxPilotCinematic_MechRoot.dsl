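; Cinematic root actor for the mech in the ToyFox pilot cinematic, which anchors itself to the giant player pawn.
; Its construction script finds all giant player pawns and, if the first is valid, converts a local offset's location and rotation into that pawn's space.
; It moves itself to the resulting world pose, then attaches to the pawn with keep-world rules.
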(fn ConstructionScript
  (bind (_localtransformoffset_location _localtransformoffset_rotation) (Math|Transform|BreakTransform (Variables|GetLocalTransformOffset_Location)))
  (bind _outactors (Actor|GetAllActorsOfClass "/ToyFoxGameplay/Gameplay/B_ToyFoxGiantPlayerPawn.B_ToyFoxGiantPlayerPawn_C"))
  (if (Utilities|Array|IsValidIndex _outactors)
    (Utilities|IsValid (Utilities|Array|Get(acopy) _outactors)
      (:Is%20Valid
        (Transformation|SetActorLocationAndRotation (Math|Transform|TransformLocation (Transformation|GetActorTransform (Utilities|Array|Get(acopy) _outactors)) _localtransformoffset_location) (Math|Transform|TransformRotation (Transformation|GetActorTransform (Utilities|Array|Get(acopy) _outactors)) _localtransformoffset_rotation))
        (Transformation|AttachActorToActor (Utilities|Array|Get(acopy) _outactors) "None" "KeepWorld" "KeepWorld" "KeepWorld")))))
