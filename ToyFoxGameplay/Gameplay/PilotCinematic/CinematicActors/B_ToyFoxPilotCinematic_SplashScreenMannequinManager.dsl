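; Splash-screen mannequin manager for the ToyFox pilot cinematic that dresses up to four mannequins as the local player's squad.
; BeginPlay gathers living squadmates by squad index, then SetupMannequins arranges child-actor mannequins by count and copies character and backpack cosmetics with variants.
; The construction script previews layouts in-editor using a debug count.
; SetupMannequinCosmetics, uncalled by the other graphs here, staggers dressing one member per tick through a self-rescheduling timer.
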(event EventBeginPlay
  (bind _teammembers (Variables|Transient|GetTeamMembers))
  ;[0 0.15 0.03648 0.5] BeginPlay: Populating the mannequin setup.
  (|CallParentFunction)
  (bind _returnvalue (Fortnite|GetFirstLocalFortPlayerController true))
  (Utilities|IsValid _returnvalue
    (:Is%20Valid
      (bind _asfort_player_state_athena (Utilities|Casting|CastToFortPlayerStateAthena (Variables|/Script/Engine.Controller|GetPlayerState _returnvalue)))
      (bind _returnvalue_2 (Class|FortPlayerStateAthena|GetSquadId _asfort_player_state_athena))
      (bind _returnvalue_3 (Fortnite|Athena|GameState|GetGameStateAthena))
      (for _array_element (Variables|/Script/Engine.GameStateBase|GetPlayerArray _returnvalue_3)
        (bind _asfort_player_state_athena_2 (Utilities|Casting|CastToFortPlayerStateAthena _array_element))
        (bind _returnvalue_4 (Class|FortPlayerStateAthena|GetSquadId _asfort_player_state_athena_2))
        (if (== _returnvalue_4 _returnvalue_2)
          (bind _asfort_player_pawn (Utilities|Casting|CastToFortPlayerPawn (PlayerState|GetPawn _asfort_player_state_athena_2)))
          (if (not (FortPawn|IsDead _asfort_player_pawn))
            (Utilities|Array|SetArrayElem (Variables|Transient|GetTeamMembers) (Class|FortPlayerStateAthena|GetSquadMemberIndex _asfort_player_state_athena_2) _asfort_player_state_athena_2 true))))
      (Utilities|Array|RemoveItem _teammembers) ; Remove empty spaces if any.
      (CallFunction|SetupMannequins (Utilities|Array|Length _teammembers) _teammembers))))

(fn SetupMannequinCosmetics
  (bind _teammemberindex (Variables|Transient|GetTeamMemberIndex))
  (bind _teammembers (Variables|Transient|GetTeamMembers))
  (bind _output (Utilities|Array|Get(acopy) _teammembers _teammemberindex))
  (bind _returnvalue (PlayerState|GetCurrentPawn _output))
  (bind _mannequinchildactor_04 (Variables|Default|GetMannequinChildActor_04))
  (bind _mannequinchildactor_03 (Variables|Default|GetMannequinChildActor_03))
  (bind _mannequinchildactor_02 (Variables|Default|GetMannequinChildActor_02))
  (bind _mannequinchildactor_01 (Variables|Default|GetMannequinChildActor_01))
  (bind _array (Utilities|Array|MakeArray _mannequinchildactor_01 _mannequinchildactor_02 _mannequinchildactor_03 _mannequinchildactor_04 0))
  (bind _output_2 (Utilities|Array|Get(acopy) _array (Variables|Transient|GetTeamMemberIndex)))
  (bind _teammemberindex_2 (Variables|Transient|GetTeamMemberIndex))
  (if (and (Utilities|Array|IsValidIndex (Variables|Transient|GetTeamMembers) (Variables|Transient|GetTeamMemberIndex)) (Utilities|Array|IsValidIndex _array _teammemberindex_2))
    (Utilities|FlowControl|Sequence
      (:then_0
        (Utilities|IsValid _output
          (:Is%20Valid
            (Utilities|IsValid _returnvalue
              (:Is%20Valid
                (Rendering|SetVisibility _output_2 true false)
                (bind _childactor (Variables|/Script/Engine.ChildActorComponent|GetChildActor _output_2)
                  (:then
                    (bind _asb_toy_fox_pilot_cinematic_mannequin (Utilities|Casting|CastToB_ToyFoxPilotCinematicMannequin _childactor))
                    (Customization|SetAthenaBackpackandCharacterDefandVariants _asb_toy_fox_pilot_cinematic_mannequin (CosmeticLoadout|GetItemDefinitionTypefromLoadoutwithActor _returnvalue "(TagName=\"LoadoutSlot.Backpack\")" "/Script/FortniteGame.AthenaBackpackItemDefinition") (CosmeticLoadout|GetActiveVariantsForItemFromActor (CosmeticLoadout|GetItemDefinitionTypefromLoadoutwithActor _returnvalue "(TagName=\"LoadoutSlot.Backpack\")" "/Script/FortniteGame.AthenaBackpackItemDefinition") _returnvalue) (CosmeticLoadout|GetItemDefinitionTypefromLoadoutwithActor _returnvalue "(TagName=\"LoadoutSlot.Character\")" "/Script/FortniteGame.AthenaCharacterItemDefinition") (CosmeticLoadout|GetActiveVariantsForItemFromActor (CosmeticLoadout|GetItemDefinitionTypefromLoadoutwithActor _returnvalue "(TagName=\"LoadoutSlot.Character\")" "/Script/FortniteGame.AthenaCharacterItemDefinition") _returnvalue)))
                  (:else)))))))
      (:then_1
        (Math|Integer|IncrementInt (Variables|Transient|GetTeamMemberIndex)
          (:%20%20
            (Utilities|Time|SetTimerforNextTickbyEvent (EventDispatchers|CreateEvent :!event SetupMannequinCosmetics))))))))

(fn SetupMannequins
  (params (Count :type Integer) (TeamMembers :type Array%20of%20Fort%20Player%20State%20Athena%20Object%20References))
  (bind _mannequinchildactor_04 (Variables|Default|GetMannequinChildActor_04))
  (bind _mannequinchildactor_03 (Variables|Default|GetMannequinChildActor_03))
  (bind _mannequinchildactor_02 (Variables|Default|GetMannequinChildActor_02))
  (bind _mannequinchildactor_01 (Variables|Default|GetMannequinChildActor_01))
  (bind _array (Utilities|Array|MakeArray _mannequinchildactor_01 _mannequinchildactor_02 _mannequinchildactor_03 _mannequinchildactor_04 0))
  (Utilities|FlowControl|Sequence
    (:then_0
      (Rendering|SetVisibility (Variables|Default|GetMannequinChildActor_01) false false)
      (switch int (Math|Integer|Clamp(Integer) Count 1 4)
        (:1
          (Transformation|SetRelativeLocation (Variables|Default|GetMannequinChildActor_01) "4.000000,-80.000000,0.000000" false false)
          (Rendering|SetVisibility (Variables|Default|GetMannequinChildActor_01) true false))
        (:2
          (Transformation|SetRelativeLocation (Variables|Default|GetMannequinChildActor_01) "-70.000000,-80.000000,0.000000" false false)
          (Transformation|SetRelativeLocation (Variables|Default|GetMannequinChildActor_02) "70.000000,-60.000000,0.000000" false false)
          (Rendering|SetVisibility (Variables|Default|GetMannequinChildActor_01) true false))
        (:3
          (Transformation|SetRelativeLocation (Variables|Default|GetMannequinChildActor_01) "4.000000,-80.000000,0.000000" false false)
          (Transformation|SetRelativeLocation (Variables|Default|GetMannequinChildActor_02) "-115.000000,-40.000000,0.000000" false false)
          (Transformation|SetRelativeLocation (Variables|Default|GetMannequinChildActor_03) "115.000000,-40.000000,0.000000" false false)
          (Rendering|SetVisibility (Variables|Default|GetMannequinChildActor_03) true false))
        (:4
          (Transformation|SetRelativeLocation (Variables|Default|GetMannequinChildActor_01) "-50.000000,-40.000000,0.000000" false false)
          (Transformation|SetRelativeLocation (Variables|Default|GetMannequinChildActor_02) "-160.000000,20.000000,0.000000" false false)
          (Transformation|SetRelativeLocation (Variables|Default|GetMannequinChildActor_03) "50.000000,-20.000000,0.000000" false false)
          (Transformation|SetRelativeLocation (Variables|Default|GetMannequinChildActor_04) "160.000000,20.000000,0.000000" false false)
          (Rendering|SetVisibility (Variables|Default|GetMannequinChildActor_01) true false))))
    (:then_1
      (for (_array_element _array_index) TeamMembers
        (Utilities|IsValid _array_element
          (:Is%20Valid
            (if (Utilities|Array|IsValidIndex _array _array_index)
              (Utilities|IsValid (PlayerState|GetCurrentPawn _array_element)
                (:Is%20Valid
                  (bind _childactor (Variables|/Script/Engine.ChildActorComponent|GetChildActor (Utilities|Array|Get(acopy) _array _array_index))
                    (:then
                      (bind _asb_toy_fox_pilot_cinematic_mannequin (Utilities|Casting|CastToB_ToyFoxPilotCinematicMannequin _childactor))
                      (Customization|SetAthenaBackpackandCharacterDefandVariants _asb_toy_fox_pilot_cinematic_mannequin (CosmeticLoadout|GetItemDefinitionTypefromLoadoutwithActor (PlayerState|GetCurrentPawn _array_element) "(TagName=\"LoadoutSlot.Backpack\")" "/Script/FortniteGame.AthenaBackpackItemDefinition") (CosmeticLoadout|GetActiveVariantsForItemFromActor (CosmeticLoadout|GetItemDefinitionTypefromLoadoutwithActor (PlayerState|GetCurrentPawn _array_element) "(TagName=\"LoadoutSlot.Backpack\")" "/Script/FortniteGame.AthenaBackpackItemDefinition") (PlayerState|GetCurrentPawn _array_element)) (CosmeticLoadout|GetItemDefinitionTypefromLoadoutwithActor (PlayerState|GetCurrentPawn _array_element) "(TagName=\"LoadoutSlot.Character\")" "/Script/FortniteGame.AthenaCharacterItemDefinition") (CosmeticLoadout|GetActiveVariantsForItemFromActor (CosmeticLoadout|GetItemDefinitionTypefromLoadoutwithActor (PlayerState|GetCurrentPawn _array_element) "(TagName=\"LoadoutSlot.Character\")" "/Script/FortniteGame.AthenaCharacterItemDefinition") (PlayerState|GetCurrentPawn _array_element))))
                    (:else)))))))))))

(fn ConstructionScript
  (if (Editor|InEditor)
    (CallFunction|SetupMannequins (Variables|Default|GetDebugMannequinCount))))
